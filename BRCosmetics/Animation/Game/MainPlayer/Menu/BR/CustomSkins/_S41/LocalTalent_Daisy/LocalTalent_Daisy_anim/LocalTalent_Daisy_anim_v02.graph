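# Level Sequence for authoring the LocalTalent_Daisy custom skin's main-player menu animation, version 02, at 30 fps.
# One spawned FortPlayerMannequin at the origin carries MR_Biped_v02 body and Fortnite_3L_Face_CtrlRig tracks, with keyed blinks, brows, gaze, jaw and mouth corners.
# Three more mannequins, offset along X and turned to different yaws, play skeletal animation sections over frames 0–216.
# A Time Warp track sits at sequence level.
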LevelSequence /BRCosmetics/Animation/Game/MainPlayer/Menu/BR/CustomSkins/_S41/LocalTalent_Daisy/LocalTalent_Daisy_anim/LocalTalent_Daisy_anim_v02
bindings [
 mannequin3 {
  id=#1
  possessed_class=FortPlayerMannequin
  tracks [
   {type=MovieSceneSpawnTrack, sections [{type=MovieSceneSpawnSection, channels [None {type=MovieSceneScriptingBoolChannel, default=true}]}]}
   {
    type=MovieScene3DTransformTrack
    sections [{
      type=MovieScene3DTransformSection
      channels [
       Location.X {type=MovieSceneScriptingDoubleChannel, default=92.5008}
       Location.Y {type=MovieSceneScriptingDoubleChannel, default=0}
       Location.Z {type=MovieSceneScriptingDoubleChannel, default=0}
       Rotation.X {type=MovieSceneScriptingDoubleChannel, default=0}
       Rotation.Y {type=MovieSceneScriptingDoubleChannel, default=0}
       Rotation.Z {type=MovieSceneScriptingDoubleChannel, default=-81.0452}
       Scale.X {type=MovieSceneScriptingDoubleChannel, default=1}
       Scale.Y {type=MovieSceneScriptingDoubleChannel, default=1}
       Scale.Z {type=MovieSceneScriptingDoubleChannel, default=1}
      ]
     }]
   }
   {type=MovieSceneSkeletalAnimationTrack, sections [{type=MovieSceneSkeletalAnimationSection, end_frame=216, start_frame=0}]}
  ]
 }
 mannequin2 {
  id=#2
  possessed_class=FortPlayerMannequin
  tracks [
   {type=MovieSceneSpawnTrack, sections [{type=MovieSceneSpawnSection, channels [None {type=MovieSceneScriptingBoolChannel, default=true}]}]}
   {
    type=MovieScene3DTransformTrack
    sections [{
      type=MovieScene3DTransformSection
      channels [
       Location.X {type=MovieSceneScriptingDoubleChannel, default=-83.2255}
       Location.Y {type=MovieSceneScriptingDoubleChannel, default=0}
       Location.Z {type=MovieSceneScriptingDoubleChannel, default=0}
       Rotation.X {type=MovieSceneScriptingDoubleChannel, default=0}
       Rotation.Y {type=MovieSceneScriptingDoubleChannel, default=0}
       Rotation.Z {type=MovieSceneScriptingDoubleChannel, default=77.4171}
       Scale.X {type=MovieSceneScriptingDoubleChannel, default=1}
       Scale.Y {type=MovieSceneScriptingDoubleChannel, default=1}
       Scale.Z {type=MovieSceneScriptingDoubleChannel, default=1}
      ]
     }]
   }
   {type=MovieSceneSkeletalAnimationTrack, sections [{type=MovieSceneSkeletalAnimationSection, end_frame=216, start_frame=0}]}
  ]
 }
 mannequin4 {
  id=#3
  possessed_class=FortPlayerMannequin
  tracks [
   {type=MovieSceneSpawnTrack, sections [{type=MovieSceneSpawnSection, channels [None {type=MovieSceneScriptingBoolChannel, default=true}]}]}
   {
    type=MovieScene3DTransformTrack
    sections [{
      type=MovieScene3DTransformSection
      channels [
       Location.X {type=MovieSceneScriptingDoubleChannel, default=169.7313}
       Location.Y {type=MovieSceneScriptingDoubleChannel, default=0}
       Location.Z {type=MovieSceneScriptingDoubleChannel, default=0}
       Rotation.X {type=MovieSceneScriptingDoubleChannel, default=0}
       Rotation.Y {type=MovieSceneScriptingDoubleChannel, default=0}
       Rotation.Z {type=MovieSceneScriptingDoubleChannel, default=180}
       Scale.X {type=MovieSceneScriptingDoubleChannel, default=1}
       Scale.Y {type=MovieSceneScriptingDoubleChannel, default=1}
       Scale.Z {type=MovieSceneScriptingDoubleChannel, default=1}
      ]
     }]
   }
   {type=MovieSceneSkeletalAnimationTrack, sections [{type=MovieSceneSkeletalAnimationSection, end_frame=216, start_frame=0}]}
  ]
 }
 mannequin {
  id=#4
  possessed_class=FortPlayerMannequin
  tracks [
   {type=MovieSceneSpawnTrack, sections [{type=MovieSceneSpawnSection, channels [None {type=MovieSceneScriptingBoolChannel, default=true}]}]}
   {
    type=MovieScene3DTransformTrack
    sections [{
      type=MovieScene3DTransformSection
      channels [
       Location.X {type=MovieSceneScriptingDoubleChannel, default=0}
       Location.Y {type=MovieSceneScriptingDoubleChannel, default=0}
       Location.Z {type=MovieSceneScriptingDoubleChannel, default=0}
       Rotation.X {type=MovieSceneScriptingDoubleChannel, default=0}
       Rotation.Y {type=MovieSceneScriptingDoubleChannel, default=0}
       Rotation.Z {type=MovieSceneScriptingDoubleChannel, default=0}
       Scale.X {type=MovieSceneScriptingDoubleChannel, default=1}
       Scale.Y {type=MovieSceneScriptingDoubleChannel, default=1}
       Scale.Z {type=MovieSceneScriptingDoubleChannel, default=1}
      ]
     }]
   }
   {type=MovieSceneControlRigParameterTrack, display_name=MR_Biped_v02, sections [{type=MovieSceneControlRigParameterSection}]}
   {
    type=MovieSceneControlRigParameterTrack
    display_name=Fortnite_3L_Face_CtrlRig
    sections [{
      type=MovieSceneControlRigParameterSection
      channels [
       CTRL_C_eyesAim.Location.X {type=MovieSceneScriptingFloatChannel, default=0}
       CTRL_C_eyesAim.Location.Y {type=MovieSceneScriptingFloatChannel, default=0}
       CTRL_C_eyesAim.Location.Z {type=MovieSceneScriptingFloatChannel, default=0}
       CTRL_C_eyesAim.Rotation.X {type=MovieSceneScriptingFloatChannel, default=0}
       CTRL_C_eyesAim.Rotation.Y {type=MovieSceneScriptingFloatChannel, default=0}
       CTRL_C_eyesAim.Rotation.Z {type=MovieSceneScriptingFloatChannel, default=0}
       CTRL_C_eyesAim.Scale.X {type=MovieSceneScriptingFloatChannel, default=1}
       CTRL_C_eyesAim.Scale.Y {type=MovieSceneScriptingFloatChannel, default=1}
       CTRL_C_eyesAim.Scale.Z {type=MovieSceneScriptingFloatChannel, default=1}
       CTRL_L_eyeAim.Location.X {type=MovieSceneScriptingFloatChannel, default=0}
       CTRL_L_eyeAim.Location.Y {type=MovieSceneScriptingFloatChannel, default=0}
       CTRL_L_eyeAim.Location.Z {type=MovieSceneScriptingFloatChannel, default=0}
       CTRL_L_eyeAim.Rotation.X {type=MovieSceneScriptingFloatChannel, default=0}
       CTRL_L_eyeAim.Rotation.Y {type=MovieSceneScriptingFloatChannel, default=0}
       CTRL_L_eyeAim.Rotation.Z {type=MovieSceneScriptingFloatChannel, default=0}
       CTRL_L_eyeAim.Scale.X {type=MovieSceneScriptingFloatChannel, default=1}
       CTRL_L_eyeAim.Scale.Y {type=MovieSceneScriptingFloatChannel, default=1}
       CTRL_L_eyeAim.Scale.Z {type=MovieSceneScriptingFloatChannel, default=1}
       CTRL_R_eyeAim.Location.X {type=MovieSceneScriptingFloatChannel, default=0}
       CTRL_R_eyeAim.Location.Y {type=MovieSceneScriptingFloatChannel, default=0}
       CTRL_R_eyeAim.Location.Z {type=MovieSceneScriptingFloatChannel, default=0}
       CTRL_R_eyeAim.Rotation.X {type=MovieSceneScriptingFloatChannel, default=0}
       CTRL_R_eyeAim.Rotation.Y {type=MovieSceneScriptingFloatChannel, default=0}
       CTRL_R_eyeAim.Rotation.Z {type=MovieSceneScriptingFloatChannel, default=0}
       CTRL_R_eyeAim.Scale.X {type=MovieSceneScriptingFloatChannel, default=1}
       CTRL_R_eyeAim.Scale.Y {type=MovieSceneScriptingFloatChannel, default=1}
       CTRL_R_eyeAim.Scale.Z {type=MovieSceneScriptingFloatChannel, default=1}
       CTRL_convergenceSwitch {type=MovieSceneScriptingFloatChannel, default=0}
       CTRL_Master.Location.X {type=MovieSceneScriptingFloatChannel, default=0}
       CTRL_Master.Location.Y {type=MovieSceneScriptingFloatChannel, default=0}
       CTRL_Master.Location.Z {type=MovieSceneScriptingFloatChannel, default=0}
       CTRL_Master.Rotation.X {type=MovieSceneScriptingFloatChannel, default=0}
       CTRL_Master.Rotation.Y {type=MovieSceneScriptingFloatChannel, default=0}
       CTRL_Master.Rotation.Z {type=MovieSceneScriptingFloatChannel, default=0}
       CTRL_Master.Scale.X {type=MovieSceneScriptingFloatChannel, default=1}
       CTRL_Master.Scale.Y {type=MovieSceneScriptingFloatChannel, default=1}
       CTRL_Master.Scale.Z {type=MovieSceneScriptingFloatChannel, default=1}
       CTRL_C_jaw.X {type=MovieSceneScriptingFloatChannel, default=0, const [0,54]}
       CTRL_C_jaw.Y {
        type=MovieSceneScriptingFloatChannel
        default=0
        thinned_from=54
        keys [[0,0],[120,0],[124,0.3612],[128,0.5237],[132,0.5779],[170,0.5779],[174,0.4334],[178,0.1084],[182,0],[212,0]]
       }
       CTRL_C_jaw_openExtreme {type=MovieSceneScriptingFloatChannel, default=0, const [0,54]}
       CTRL_C_jaw_fwdBack {type=MovieSceneScriptingFloatChannel, default=0, const [0,54]}
       CTRL_R_jaw_ChinRaiseD {type=MovieSceneScriptingFloatChannel, default=0, const [0,54]}
       CTRL_R_jaw_ChinRaiseU {type=MovieSceneScriptingFloatChannel, default=0, const [0,54]}
       CTRL_L_jaw_ChinRaiseU {type=MovieSceneScriptingFloatChannel, default=0, const [0,54]}
       CTRL_L_jaw_ChinRaiseD {type=MovieSceneScriptingFloatChannel, default=0, const [0,54]}
       CTRL_R_mouth_lipsBlow {type=MovieSceneScriptingFloatChannel, default=0, const [0,54]}
       CTRL_L_mouth_lipsTogetherU {type=MovieSceneScriptingFloatChannel, default=0, const [0,54]}
       CTRL_L_mouth_lipsTogetherD {type=MovieSceneScriptingFloatChannel, default=0, const [0,54]}
       CTRL_R_mouth_lipsTogetherU {type=MovieSceneScriptingFloatChannel, default=0, const [0,54]}
       CTRL_L_mouth_lipsBlow {type=MovieSceneScriptingFloatChannel, default=0, const [0,54]}
       CTRL_R_mouth_lipsTogetherD {type=MovieSceneScriptingFloatChannel, default=0, const [0,54]}
       CTRL_C_tongue_roll.X {type=MovieSceneScriptingFloatChannel, default=0, const [0,54]}
       CTRL_C_tongue_roll.Y {type=MovieSceneScriptingFloatChannel, default=0, const [0,54]}
       CTRL_C_tongue.X {type=MovieSceneScriptingFloatChannel, default=0, const [0,54]}
       CTRL_C_tongue.Y {type=MovieSceneScriptingFloatChannel, default=0, const [0,54]}
       CTRL_C_tongue_narrowWide {type=MovieSceneScriptingFloatChannel, default=0, const [0,54]}
       CTRL_C_tongue_inOut {type=MovieSceneScriptingFloatChannel, default=0, const [0,54]}
       CTRL_R_nose.X {type=MovieSceneScriptingFloatChannel, default=0, const [0,54]}
       CTRL_R_nose.Y {type=MovieSceneScriptingFloatChannel, default=0, const [0,54]}
       CTRL_L_nose.X {type=MovieSceneScriptingFloatChannel, default=0, const [0,54]}
       CTRL_L_nose.Y {type=MovieSceneScriptingFloatChannel, default=0, const [0,54]}
       CTRL_L_eye.X {type=MovieSceneScriptingFloatChannel, default=0, const [0,54]}
       CTRL_L_eye.Y {type=MovieSceneScriptingFloatChannel, default=0, const [0,54]}
       CTRL_L_eye_blink {
        type=MovieSceneScriptingFloatChannel
        default=0
        thinned_from=43
        keys [[0,0],[11,0],[12,0.125],[13,0.8503],[14,1],[17,1],[19,0.75],[21,0.0938],[24,0],[43,0],[44,0.125],[45,0.8503],[46,1],[144,1],[146,0.75],[148,0.0938],[151,0],[176,0],[177,0.125],[178,0.8503],[179,1],[182,1],[184,0.75],[186,0.0938],[189,0],[212,0]]
       }
       CTRL_R_eye.X {type=MovieSceneScriptingFloatChannel, default=0, const [0,54]}
       CTRL_R_eye.Y {type=MovieSceneScriptingFloatChannel, default=0, const [0,54]}
       CTRL_R_eye_blink {
        type=MovieSceneScriptingFloatChannel
        default=0
        thinned_from=43
        keys [[0,0],[11,0],[12,0.125],[13,0.8503],[14,1],[17,1],[19,0.75],[21,0.0938],[24,0],[43,0],[44,0.125],[45,0.8503],[46,1],[144,1],[146,0.75],[148,0.0938],[151,0],[176,0],[177,0.125],[178,0.8503],[179,1],[182,1],[184,0.75],[186,0.0938],[189,0],[212,0]]
       }
       CTRL_C_eye.X {
        type=MovieSceneScriptingFloatChannel
        default=0
        thinned_from=54
        keys [[0,0.112],[4,0.112],[8,0.0929],[12,-0.1418],[16,-0.2108],[20,-0.2338],[40,-0.2338],[44,-0.2131],[48,-0.1411],[52,-0.0096],[56,0.1664],[60,0.3554],[64,0.3824],[76,0.3824],[80,0.3625],[84,0.3293],[88,0.2552],[92,0.0331],[96,-0.0656],[112,-0.0656],[116,0.0947],[120,0.3086],[144,0.3086],[148,0.2151],[152,0.0374],[156,-0.0656],[172,-0.0656],[176,0.0082],[180,0.1742],[184,0.2295],[196,0.2295],[200,0.2152],[204,0.1212],[208,0.112],[212,0.112]]
       }
       CTRL_C_eye.Y {
        type=MovieSceneScriptingFloatChannel
        default=0
        thinned_from=54
        keys [[0,-0.4714],[4,-0.4714],[8,-0.4461],[12,-0.3703],[16,-0.2787],[20,-0.2481],[24,-0.2671],[28,-0.317],[32,-0.3242],[40,-0.3242],[48,-0.4714],[52,-0.4536],[56,-0.3976],[60,-0.3374],[64,-0.3288],[76,-0.3288],[80,-0.24],[84,-0.0919],[88,-0.1631],[92,-0.3765],[96,-0.4714],[212,-0.4714]]
       }
       CTRL_C_eyesConvergence {type=MovieSceneScriptingFloatChannel, default=0, const [0,54]}
       CTRL_L_eye_cheekRaise {type=MovieSceneScriptingFloatChannel, default=0, const [0,54]}
       CTRL_L_eye_squintInner {type=MovieSceneScriptingFloatChannel, default=0, const [0,54]}
       CTRL_L_nose_wrinkleUpper {type=MovieSceneScriptingFloatChannel, default=0, const [0,54]}
       CTRL_R_nose_wrinkleUpper {type=MovieSceneScriptingFloatChannel, default=0, const [0,54]}
       CTRL_R_eye_squintInner {type=MovieSceneScriptingFloatChannel, default=0, const [0,54]}
       CTRL_R_eye_cheekRaise {type=MovieSceneScriptingFloatChannel, default=0, const [0,54]}
       CTRL_C_eyelidsReduce {type=MovieSceneScriptingFloatChannel, default=0}
       CTRL_R_brow_lateral {type=MovieSceneScriptingFloatChannel, default=0, const [0,54]}
       CTRL_L_brow_lateral {type=MovieSceneScriptingFloatChannel, default=0, const [0,54]}
       CTRL_R_brow_down {
        type=MovieSceneScriptingFloatChannel
        default=0
        thinned_from=43
        keys [[0,0],[11,0],[12,0.1104],[13,0.7509],[14,0.8831],[17,0.8831],[19,0.6623],[21,0.0828],[24,0],[43,0],[44,0.1104],[45,0.7509],[46,0.8831],[144,0.8831],[146,0.6623],[148,0.0828],[151,0],[176,0],[177,0.1104],[178,0.7509],[179,0.8831],[182,0.8831],[184,0.6623],[186,0.0828],[189,0],[212,0]]
       }
       CTRL_L_brow_down {
        type=MovieSceneScriptingFloatChannel
        default=0
        thinned_from=43
        keys [[0,0],[11,0],[12,0.1104],[13,0.7509],[14,0.8831],[17,0.8831],[19,0.6623],[21,0.0828],[24,0],[43,0],[44,0.1104],[45,0.7509],[46,0.8831],[144,0.8831],[146,0.6623],[148,0.0828],[151,0],[176,0],[177,0.1104],[178,0.7509],[179,0.8831],[182,0.8831],[184,0.6623],[186,0.0828],[189,0],[212,0]]
       }
       CTRL_L_brow_raiseOut {type=MovieSceneScriptingFloatChannel, default=0, const [0,54]}
       CTRL_L_brow_raiseIn {type=MovieSceneScriptingFloatChannel, default=0, const [0.7747,54]}
       CTRL_R_brow_raiseIn {type=MovieSceneScriptingFloatChannel, default=0, const [0.7747,54]}
       CTRL_R_brow_raiseOut {type=MovieSceneScriptingFloatChannel, default=0, const [0,54]}
       CTRL_lookAtSwitch {type=MovieSceneScriptingFloatChannel, default=0, const [0,54]}
       CTRL_C_teethD.X {type=MovieSceneScriptingFloatChannel, default=0, const [0,54]}
       CTRL_C_teethD.Y {type=MovieSceneScriptingFloatChannel, default=0, const [0,54]}
       CTRL_C_teethU.X {type=MovieSceneScriptingFloatChannel, default=0, const [0,54]}
       CTRL_C_teethU.Y {type=MovieSceneScriptingFloatChannel, default=0, const [0,54]}
       CTRL_C_teeth_fwdBackU {type=MovieSceneScriptingFloatChannel, default=0, const [0,54]}
       CTRL_L_mouth_corner.X {type=MovieSceneScriptingFloatChannel, default=0, thinned_from=54, keys [[0,1],[56,1],[60,-0.2967],[64,1],[212,1]]}
       CTRL_L_mouth_corner.Y {type=MovieSceneScriptingFloatChannel, default=0, const [1,54]}
       CTRL_R_mouth_corner.X {
        type=MovieSceneScriptingFloatChannel
        default=0
        thinned_from=54
        keys [[0,1],[56,1],[60,0.2133],[64,1],[132,1],[136,0.6738],[140,1],[212,1]]
       }
       CTRL_R_mouth_corner.Y {type=MovieSceneScriptingFloatChannel, default=0, const [1,54]}
       CTRL_C_teeth_fwdBackD {type=MovieSceneScriptingFloatChannel, default=0, const [0,54]}
       CTRL_R_mouth_cornerSharpnessD {type=MovieSceneScriptingFloatChannel, default=0, const [0,54]}
       CTRL_R_mouth_cornerSharpnessU {type=MovieSceneScriptingFloatChannel, default=0, const [0,54]}
       CTRL_L_eye_eyelidU.X {type=MovieSceneScriptingFloatChannel, default=0, const [0,54]}
       CTRL_L_eye_eyelidU.Y {type=MovieSceneScriptingFloatChannel, default=0, const [0,54]}
       CTRL_L_mouth_cornerSharpnessD {type=MovieSceneScriptingFloatChannel, default=0, const [0,54]}
       CTRL_L_eye_eyelidD.X {type=MovieSceneScriptingFloatChannel, default=0, const [0,54]}
       CTRL_L_eye_eyelidD.Y {type=MovieSceneScriptingFloatChannel, default=0, const [0,54]}
       CTRL_L_mouth_cornerSharpnessU {type=MovieSceneScriptingFloatChannel, default=0, const [0,54]}
       CTRL_R_mouth_lipsTowardsTeethD {type=MovieSceneScriptingFloatChannel, default=0, const [0,54]}
       CTRL_L_mouth_lipsTowardsTeethD {type=MovieSceneScriptingFloatChannel, default=0, const [0,54]}
       CTRL_R_eye_eyelidU.X {type=MovieSceneScriptingFloatChannel, default=0, const [0,54]}
       CTRL_R_eye_eyelidU.Y {type=MovieSceneScriptingFloatChannel, default=0, const [0,54]}
       CTRL_L_mouth_lipsTowardsTeethU {type=MovieSceneScriptingFloatChannel, default=0, const [0,54]}
       CTRL_R_eye_eyelidD.X {type=MovieSceneScriptingFloatChannel, default=0, const [0,54]}
       CTRL_R_eye_eyelidD.Y {type=MovieSceneScriptingFloatChannel, default=0, const [0,54]}
       CTRL_R_mouth_lipsTowardsTeethU {type=MovieSceneScriptingFloatChannel, default=0}
       CTRL_faceGUIfollowHead {type=MovieSceneScriptingFloatChannel, default=1}
       CTRL_eyesAimFollowHead {type=MovieSceneScriptingFloatChannel, default=0}
       CTRL_C_nose.X {type=MovieSceneScriptingFloatChannel, default=0, const [0,54]}
       CTRL_C_nose.Y {type=MovieSceneScriptingFloatChannel, default=0, const [0,54]}
       CTRL_C_teeth_rotateU {type=MovieSceneScriptingFloatChannel, default=-1, const [-1,11]}
       CTRL_C_teeth_rotateD {type=MovieSceneScriptingFloatChannel, default=0}
       CTRL_L_brow_inner {type=MovieSceneScriptingFloatChannel, default=0, const [0,54]}
       CTRL_R_brow_inner {type=MovieSceneScriptingFloatChannel, default=0, const [0,54]}
       CTRL_L_brow_outer {type=MovieSceneScriptingFloatChannel, default=0, const [0,54]}
       CTRL_R_brow_outer {type=MovieSceneScriptingFloatChannel, default=0, const [0,54]}
       CTRL_C_mouth.X {type=MovieSceneScriptingFloatChannel, default=0, const [0,54]}
       CTRL_C_mouth.Y {type=MovieSceneScriptingFloatChannel, default=0, const [0,54]}
       CTRL_L_mouth_upperLipRaise {type=MovieSceneScriptingFloatChannel, default=0, const [0,54]}
       CTRL_L_mouth_cornerPull {
        type=MovieSceneScriptingFloatChannel
        default=0
        thinned_from=54
        keys [[0,0.1119],[120,0.1119],[124,0.4512],[128,0.6039],[132,0.6548],[148,0.6548],[152,0.5871],[156,0.5209],[160,0.5114],[170,0.5114],[174,0.4115],[178,0.1868],[182,0.1119],[212,0.1119]]
       }
       CTRL_L_mouth_dimple {type=MovieSceneScriptingFloatChannel, default=0, const [0,54]}
       CTRL_L_mouth_cornerDepress {type=MovieSceneScriptingFloatChannel, default=0, const [0,54]}
       CTRL_L_mouth_stretch {type=MovieSceneScriptingFloatChannel, default=0, const [0,54]}
       CTRL_L_mouth_lowerLipDepress {type=MovieSceneScriptingFloatChannel, default=0, const [0,54]}
       CTRL_R_mouth_lowerLipDepress {type=MovieSceneScriptingFloatChannel, default=0, const [0,54]}
       CTRL_R_mouth_stretch {type=MovieSceneScriptingFloatChannel, default=0, const [0,54]}
       CTRL_L_mouth_towardsD {type=MovieSceneScriptingFloatChannel, default=0, const [0,54]}
       CTRL_R_mouth_funnelU {type=MovieSceneScriptingFloatChannel, default=0, const [0,54]}
       CTRL_L_mouth_towardsU {type=MovieSceneScriptingFloatChannel, default=0, const [0,54]}
       CTRL_R_mouth_lipBiteD {type=MovieSceneScriptingFloatChannel, default=0, const [0,54]}
       CTRL_R_mouth_funnelD {type=MovieSceneScriptingFloatChannel, default=0, const [0,54]}
       CTRL_L_mouth_pressU {type=MovieSceneScriptingFloatChannel, default=0, const [0,54]}
       CTRL_L_mouth_funnelD {type=MovieSceneScriptingFloatChannel, default=0, const [0,54]}
       CTRL_L_mouth_suckBlow {type=MovieSceneScriptingFloatChannel, default=0, const [0,54]}
       CTRL_L_mouth_pressD {type=MovieSceneScriptingFloatChannel, default=0, const [0,54]}
       CTRL_L_mouth_funnelU {type=MovieSceneScriptingFloatChannel, default=0, const [0,54]}
       CTRL_R_mouth_towardsU {type=MovieSceneScriptingFloatChannel, default=0, const [0,54]}
       CTRL_R_mouth_purseD {type=MovieSceneScriptingFloatChannel, default=0, const [0,54]}
       CTRL_L_mouth_lipBiteU {type=MovieSceneScriptingFloatChannel, default=0, const [0,54]}
       CTRL_R_mouth_pressD {type=MovieSceneScriptingFloatChannel, default=0, const [0,54]}
       CTRL_R_mouth_pressU {type=MovieSceneScriptingFloatChannel, default=0, const [0,54]}
       CTRL_L_mouth_purseD {type=MovieSceneScriptingFloatChannel, default=0, const [0,54]}
       CTRL_R_mouth_lipBiteU {type=MovieSceneScriptingFloatChannel, default=0, const [0,54]}
       CTRL_R_mouth_suckBlow {type=MovieSceneScriptingFloatChannel, default=0, const [0,54]}
       CTRL_L_mouth_purseU {type=MovieSceneScriptingFloatChannel, default=0, const [0,54]}
       CTRL_R_mouth_towardsD {type=MovieSceneScriptingFloatChannel, default=0, const [0,54]}
       CTRL_L_mouth_lipBiteD {type=MovieSceneScriptingFloatChannel, default=0, const [0,54]}
       CTRL_R_mouth_purseU {type=MovieSceneScriptingFloatChannel, default=0, const [0,54]}
       CTRL_R_mouth_cornerDepress {type=MovieSceneScriptingFloatChannel, default=0, const [0,54]}
       CTRL_R_mouth_upperLipRaise {type=MovieSceneScriptingFloatChannel, default=0, const [0,54]}
       CTRL_R_mouth_dimple {type=MovieSceneScriptingFloatChannel, default=0, const [0,54]}
       CTRL_R_mouth_cornerPull {
        type=MovieSceneScriptingFloatChannel
        default=0
        thinned_from=54
        keys [[0,0.3003],[120,0.3003],[124,0.3575],[128,0.3833],[132,0.3919],[148,0.3919],[152,0.5054],[156,0.6164],[160,0.6322],[170,0.6322],[174,0.5492],[178,0.3626],[182,0.3003],[212,0.3003]]
       }
      ]
     }]
   }
  ]
 }
]
display_rate {denominator=1, numerator=30}
tick_resolution {denominator=1, numerator=24000}
tracks [{type=MovieSceneTimeWarpTrack, display_name=Time Warp Curve, sections [{type=MovieSceneTimeWarpSection}]}]
